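; DangerGrape AI pawn Blueprint, apparently a decoy clone judging by its death sound, that adds spawn and death presentation to its parent pawn.
; BeginPlay waits 0.2 seconds after the parent logic, then makes the character mesh visible.
; The PlayResOut_1 event attaches a disguise death particle to the mesh's chest socket, skipped on mobile.
; Death effects run the parent, then play the decoy clone death sound at the pawn's location.
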(event EventBeginPlay
  (|CallParentFunction)
  (Utilities|FlowControl|Delay 0.2)
  (Rendering|SetVisibility (Variables|Character|GetMesh) true false))

(event EventTick
  (params (DeltaSeconds :type Float%20%28single-precision%29))
  (|CallParentFunction DeltaSeconds) ; This node is disabled and will not be called.\nDrag off pins to build functionality.
)

(event Custom|PlayResOut_1
  (if (not (Fortnite|IsMobilePlatform))
    (Effects|SpawnEmitterAttached :EmitterTemplate "/Game/Athena/Items/Consumables/DangerGrape/P_Creative_Disguise_DangerGrapeDeath.P_Creative_Disguise_DangerGrapeDeath" :AttachToComponent (Variables|Character|GetMesh) :AttachPointName "fx_chest" :Scale "1.000000,1.000000,1.000000" :LocationType "KeepRelativeOffset" :bAutoDestroy true :PoolingMethod "AutoRelease" :bAutoActivate true)))

(event EventOnDeathPlayEffects
  (params (Damage :type Float%20%28single-precision%29) (DamageTags :type Gameplay%20Tag%20Container%20Structure%20%28by%20ref%29) (Momentum :type Vector :default "0, 0, 0") (HitInfo :type Hit%20Result%20Structure%20%28by%20ref%29) (InstigatedBy :type Fort%20Pawn%20Object%20Reference) (DamageCauser :type Actor%20Object%20Reference) (EffectContext :type Gameplay%20Effect%20Context%20Handle%20Structure))
  (|CallParentFunction Damage DamageTags Momentum HitInfo InstigatedBy DamageCauser EffectContext)
  (Audio|PlaySoundatLocation "/Game/Sounds/Fort_Weapons/Grenades/Decoy/DecoyGrenade_CloneDeath_Cue.DecoyGrenade_CloneDeath_Cue" (Transformation|GetActorLocation)))

(fn ConstructionScript
  (|CallParentFunction))
